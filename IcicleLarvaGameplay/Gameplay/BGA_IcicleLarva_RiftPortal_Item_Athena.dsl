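; Gameplay actor blueprint BGA_IcicleLarva_RiftPortal_Item_Athena, a rift portal item that removes its rift shortly after spawning.
; On BeginPlay it calls the parent logic, then starts a short one-shot timer bound to a custom event.
; When the timer fires, that event calls DestroyRift; the construction script only runs the parent version.
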(event EventBeginPlay
  (|CallParentFunction)
  (Utilities|Time|SetTimerbyEvent (delegate Custom|CustomDestroyRift) 1.0))

(event Custom|CustomDestroyRift
  (CallFunction|DestroyRift))

(fn ConstructionScript
  (|CallParentFunction))
