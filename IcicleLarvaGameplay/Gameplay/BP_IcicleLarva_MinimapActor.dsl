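; Minimap marker actor for the IcicleLarva item, switching from a skydrop marker to a player marker when the item is claimed.
; On BeginPlay it finds the IcicleLarva game state component and binds to its item-claimed dispatcher.
; On claim it stores the claiming pawn, binds its OnDied, destroys the skydrop special-actor component and activates the player one.
; When that pawn dies, it treats the item as defeated and destroys itself.
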(event EventBeginPlay
  (bind _returnvalue (Game|GetGameState))
  ; Begin Play
  (Utilities|IsValid _returnvalue
    (:Is%20Valid
      (bind _output_get (Variables|Default|SetIcicleLarvaGameStateComp (Actor|GetComponentbyClass _returnvalue "/IcicleLarvaGameplay/Gameplay/IcicleLarvaFortGamestateComp.IcicleLarvaFortGamestateComp_C")))
      (Utilities|IsValid _output_get
        (:Is%20Valid
          (Default|AssignOnIcicleLarvaItemClaimed _output_get (EventDispatchers|CreateEvent :!event Custom|OnIcicleLarvaItemClaimed)))))))

(event Custom|OnIcicleLarvaItemClaimed
  (params (IcicleLarvaPlayerPawn :type Fort%20Pawn%20Object%20Reference) (IcicleLarvaPlayerState :type Player%20State%20Object%20Reference))
  ; Item Claimed - Remove skydrop Minimap Actor
  (bind _output_get (Variables|Default|SetIcicleLarvaPlayerPawn IcicleLarvaPlayerPawn))
  (EventDispatchers|BindEventtoOnDied _output_get (delegate Custom|OnDied))
  (Components|DestroyComponent (Variables|Default|GetAthenaSpecialActor_Skydrop))
  (Components|Activation|SetActive (Variables|Default|GetAthenaSpecialActor_Player) true false))

(event Custom|OnDied
  (params (DamagedActor :type Actor%20Object%20Reference) (Damage :type Float%20%28single-precision%29) (InstigatedBy :type Controller%20Object%20Reference) (DamageCauser :type Actor%20Object%20Reference) (HitLocation :type Vector) (FHitComponent :type Primitive%20Component%20Object%20Reference) (BoneName :type Name) (Momentum :type Vector))
  ; Item Claimed - Remove skydrop Minimap Actor
  (CallFunction|OnIcicleLarvaItemDefeated))

(event Custom|OnIcicleLarvaItemDefeated
  ; Item Defeated
  (Actor|DestroyActor))

(fn ConstructionScript)
