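; Icicle Larva special weapon that sets its holder to max health and signals claimed and defeated notifications through a game state component.
; While equipped it adds a custom tactical sprint extension and swaps HUD keys when the flying movement tag changes.
; A server timer reports its location for the minimap.
; Non-teammates not viewing the holder get a nav marker that tracks screen clamping and shifts with distance to avoid overlapping the pawn.
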(event EventBeginPlay
  (bind _returnvalue (Game|GetGameState))
  (Utilities|IsValid _returnvalue
    (:Is%20Valid
      (Variables|Default|SetIcicleLarvaGameStateComp (Actor|GetComponentbyClass _returnvalue "/IcicleLarvaGameplay/Gameplay/IcicleLarvaFortGamestateComp.IcicleLarvaFortGamestateComp_C")))))

(event FortWeapon|EventOnEquippedAndAttached
  (|CallParentFunction)
  (Utilities|FlowControl|Sequence
    (:then_0
      ; On equip set Health to max
      (bind _asfort_pawn (Utilities|Casting|CastToFortPawn (Game|GetInstigator)))
      (bind _output_get (Variables|Default|SetFort%20Pawn _asfort_pawn))
      (FortPawn|SetHealth _output_get (ScalableFloat|GetValueatLevel (Variables|Default|GetSpecialMaxHealth)))
      ;[0.521 0.402 1 0.502] Minimap/Compass and Claimed Notification
      (Class|IcicleLarvaFortGamestateComp|UpdateIcicleLarvaItemClaimed (Variables|Default|GetIcicleLarvaGameStateComp) (Variables|Default|GetFort%20Pawn))
      (bind (_asynctaskproxy _tagcount) (Ability|Tasks|WaitGameplayTagCountChangedOnActor (Variables|Default|GetFort%20Pawn) "(TagName=\"MovementMode.IcicleLarva\")")
        (:then
          (Variables|Default|SetHudKeyAsyncAction _asynctaskproxy))
        (:TagCountChanged
          (CallFunction|SetFlyingHUDKeys (> _tagcount 0)))))
    (:then_1
      ;[0.521 0.402 1 0.502] Nav Point Screen Location Clamp State
      (if (not (or (Team|OnLocalPlayersTeam (Variables|Default|GetFort%20Pawn)) (Pawn|IsLocallyViewed (Variables|Default|GetFort%20Pawn))))
        (bind _fortwidget_playernavpoint (Variables|Default|GetFortWidget_PlayerNavPoint)
          (:then
            (bind _aswbp_icicle_larva_nav_point_marker (Utilities|Casting|CastToWBP_IcicleLarva_NavPointMarker (UserInterface|GetWidget _fortwidget_playernavpoint)))
            (Class|WBPIcicleLarvaNavPointMarker|EnableNavPointMarker _aswbp_icicle_larva_nav_point_marker)
            (Appearance|Event|BindEventtoOnVisibilityChanged _aswbp_icicle_larva_nav_point_marker (delegate Custom|OnNavPointVisibilityChanged))
            (EventDispatchers|BindEventtoOnClampStateChangedEvent _fortwidget_playernavpoint (EventDispatchers|CreateEvent :!event OnWidgetComponentClampStateChanged)))
          (:else))))
    (:then_2
      (Sprint|AddCustomSprintMMEClass (Actor|GetComponentbyClass (Pawn|GetController (Variables|Default|GetFort%20Pawn)) "/TacticalSprintGame/Gameplay/TacticalSprintControllerComponent.TacticalSprintControllerComponent_C") (Variables|Default|GetCustomTacSprintExtensionClass))
      ;[0.2664 1 0.3372 0.5] Send location to Gamestate for Minimap Reference
      (Networking|SwitchHasAuthority
        (:Authority
          (bind _returnvalue (Utilities|Time|SetTimerbyEvent (delegate Custom|SendLocationToServer) (ScalableFloat|GetValueatLevel (Variables|MinimapLocation|GetLocationUpdateFrequency)) true))
          (Variables|MinimapLocation|SetLocationUpdateTimer _returnvalue))))))

(event Custom|OnNavPointVisibilityChanged
  (params (InVisibility :type ESlateVisibility%20Enum :default "Visible"))
  ;[0.521 0.402 1 0.502] Nav Point Screen Location Clamp State
  ;; If Nav Point is visible start offset update
  (if (Utilities|Select :Index InVisibility :Visible true :Collapsed false :Hidden false :HitTestInvisible true :SelfHitTestInvisible true)
    (bind _returnvalue (Utilities|Time|SetTimerbyEvent (EventDispatchers|CreateEvent :!event UpdateNavPointWidgetOffset) 0.2 true))
    (Variables|NavPointMarker|SetNavPointOffsetTImerHandle _returnvalue)
    (else
      (Utilities|Time|ClearandInvalidateTimerbyHandle (Variables|NavPointMarker|GetNavPointOffsetTImerHandle)))))

(event Custom|SendLocationToServer
  (bind _iciclelarvagamestatecomp (Variables|Default|GetIcicleLarvaGameStateComp))
  ;[0.2664 1 0.3372 0.5] Send location to Gamestate for Minimap Reference
  (Utilities|IsValid _iciclelarvagamestatecomp
    (:Is%20Valid
      (Variables|/IcicleLarvaGameplay/Gameplay/IcicleLarvaFortGamestateComp.IcicleLarvaFortGamestateComp_C|SetIcicleLarvaLocation (Transformation|GetActorLocation) _iciclelarvagamestatecomp))))

(event FortWeapon|EventOnUnEquippedAndDetached
  (|CallParentFunction)
  (Ability|Async|EndAction (Variables|Default|GetHudKeyAsyncAction))
  (Sprint|RemoveCustomSprintMMEClass (Actor|GetComponentbyClass (Pawn|GetController (Variables|Default|GetFort%20Pawn)) "/TacticalSprintGame/Gameplay/TacticalSprintControllerComponent.TacticalSprintControllerComponent_C") (Variables|Default|GetCustomTacSprintExtensionClass))
  (Utilities|Time|ClearandInvalidateTimerbyHandle (Variables|MinimapLocation|GetLocationUpdateTimer))
  (Utilities|Time|ClearandInvalidateTimerbyHandle (Variables|NavPointMarker|GetNavPointOffsetTImerHandle))
  ;[0.5203 0.401 1 0.5] Play Defeated Notifications
  (Class|IcicleLarvaFortGamestateComp|UpdateIcicleLarvaItemDefeated (Variables|Default|GetIcicleLarvaGameStateComp)))

(event EventOnUnableToPerformAction
  (params (FailedReason :type Gameplay%20Tag%20Container%20Structure%20%28by%20ref%29) (LocalizedFailureReasonText :type Text%20%28by%20ref%29))
  (|CallParentFunction FailedReason LocalizedFailureReasonText)
  (if (GameplayTags|HasAnyTags FailedReason (Variables|Default|GetCooldownContainer) true)
    (Audio|PlaySound2D "/Game/Athena/Sounds/Items/Ability_CantUse_Cue.Ability_CantUse_Cue")))

(event EventEndPlay
  (params (EndPlayReason :type EEndPlayReason%20Enum :default "Destroyed"))
  (|CallParentFunction EndPlayReason)
  (Utilities|Time|ClearandInvalidateTimerbyHandle (Variables|MinimapLocation|GetLocationUpdateTimer))
  (Utilities|Time|ClearandInvalidateTimerbyHandle (Variables|NavPointMarker|GetNavPointOffsetTImerHandle)))

(fn GetBeamAudioActor
  (bind _asbp_beam_audio (Utilities|Casting|CastToBP_BeamAudio (Variables|/Script/Engine.ChildActorComponent|GetChildActor (Variables|Default|GetBeamAudioController)))
    (:then
      (return _asbp_beam_audio))
    (:CastFailed
      (return 0))))

(fn GetBeamOwner
  (return 0 false))

(fn OnWidgetComponentClampStateChanged
  (params (ClampToScreenDirection :type EClampToScreenDirection%20Enum :default "Left"))
  (Variables|NavPointMarker|SetNavPoint_ClampToScreenDirection ClampToScreenDirection)
  (bind _fortwidget_playernavpoint (Variables|Default|GetFortWidget_PlayerNavPoint)
    (:then
      (bind _aswbp_icicle_larva_nav_point_marker (Utilities|Casting|CastToWBP_IcicleLarva_NavPointMarker (UserInterface|GetWidget _fortwidget_playernavpoint)))
      (Default|OnClampStateChanged _aswbp_icicle_larva_nav_point_marker ClampToScreenDirection))
    (:else)))

(fn SetFlyingHUDKeys
  (params (IsFlying :type Boolean))
  (LocalVariables|SetPlayerFlying IsFlying)
  (Utilities|FlowControl|Sequence
    (:then_0
      (for _array_element (Variables|Default|GetFlyingKeyActionIndexArray)
        (Utilities|Array|Add (LocalVariables|GetVisibiltyHudKeys) (Utilities|Struct|MakeWeaponHudKeyActionVisibility _array_element (LocalVariables|GetPlayerFlying)))))
    (:then_1
      (for _array_element_2 (Variables|Default|GetGroundKeyActionIndexArray)
        (Utilities|Array|Add (LocalVariables|GetVisibiltyHudKeys) (Utilities|Struct|MakeWeaponHudKeyActionVisibility _array_element_2 (not (LocalVariables|GetPlayerFlying))))))
    (:then_2
      (WeaponHUDData|SetHudKeyActionsVisibility (LocalVariables|GetVisibiltyHudKeys)))))

(fn UpdateNavPointWidgetOffset
  (if (Utilities|Enum|Equal(Enum) (Variables|NavPointMarker|GetNavPoint_ClampToScreenDirection) "None")
    (Networking|SwitchHasAuthority
      (:Remote
        (bind _array_element (Utilities|Array|ForEachLoopwithBreak (Fortnite|GetLocalFortPlayerControllers true))
          (:LoopBody
            (Utilities|IsValid (Pawn|GetControlledPawn _array_element)
              (:Is%20Valid
                (Variables|/Script/UMG.Widget|SetRenderTransform_Translation_X (Utilities|Struct|MakeWidgetTransform (Math|Vector2D|MakeVector2D 0.0 (Math|Float|Lerp 0.0 -5e+01 (Math|Float|Clamp(Float) (Math|Float|NormalizetoRange (Math|Vector|Distance(Vector) (Transformation|GetActorLocation) (Transformation|GetActorLocation (Pawn|GetControlledPawn _array_element))) 0.0 5e+03))))) (UserInterface|GetUserWidgetObject (Variables|Default|GetFortWidget_PlayerNavPoint))) ; Prevents overlapping with pawn when increasing \ndistance between player
              ))))))))

(fn ConstructionScript
  (|CallParentFunction))
