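; Game state component that runs the Icicle Larva skydrop event and relays item status to client UI.
; On playlist ready, the server rolls spawn chance, then arms a one-shot random timer from game phase events.
; The timer picks a data-driven location for a staged skydrop with attached minimap actor, gated by an event-router toggle.
; Spawn, claim, health, damage and defeat reach UI dispatchers through client events and RepNotifies.
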(event EventBeginPlay
  (|CallParentFunction)
  (Class|FortAsyncActionCurrentPlaylistReady|PlaylistDataReadyAsync
    (:OnReady
      (Networking|SwitchHasAuthority
        (:Authority
          (bind _iseventactive (CallFunction|RollEventSpawnChance))
          (if _iseventactive
            (CallFunction|SetupSpawnTimer)))))))

(event Custom|SetupSpawnTimer
  (bind (_asynctaskproxy _payload) (GameplayEventRouter|StartListeningToStatefulEvent (GameplayEventRouter|GetGlobalEventRouter) "/Script/FortniteGame.FortGamePhaseUpdatedEvent")
    (:then
      (switch Utilities|FlowControl|Switch|SwitchonEAthenaGamePhase (BattleRoyaleGamePhaseLogic|GetGamePhase (BattleRoyaleGamePhaseLogic|GetBattleRoyaleGamePhaseLogicComponent))
        (:Aircraft
          _)
        (:SafeZones
          _)))
    (:OnEventReceived
      (switch Utilities|FlowControl|Switch|SwitchonEAthenaGamePhase (Utilities|Struct|BreakFortGamePhaseUpdatedEvent _payload)
        (:Aircraft
          _)
        (:SafeZones
          _)
        (:EndGame
          _)
        (:Count
          _))))
  (CallFunction|StartSpawnTimer) ; intention is we only start the timer at aircraft phase.\nbut by handling all cases, this lets us cheat phases in PIE\nthe do-once will prevent multiple timers from being set.
  (CallFunction|BindSkyDropEvent) ; allow the skydrop to be enabled / disabled\nex: by a live event
)

(event Custom|IcicleLarvaSpawnAtLocation
  (params (x :type Integer) (y :type Integer))
  (bind _output_get (Variables|SetCustomTransform_Location_X (Math|Transform|MakeTransform (Math|Vector|MakeVector (Math|Conversions|ToFloat(Integer) x) (Math|Conversions|ToFloat(Integer) y) 4e+04))))
  (CallFunction|SkydropSpawnatTransform _output_get))

(event Custom|StartSpawnTimer
  ;[1 1 1] Spawn Timer
  (Utilities|FlowControl|DoOnce false
    (:Completed
      (bind _returnvalue (Utilities|Time|SetTimerbyEvent (delegate Custom|SpawnItem) (Math|Random|RandomFloatinRange (ScalableFloat|GetValueatLevel (Variables|GameData|GetSpawnDelayMin)) (ScalableFloat|GetValueatLevel (Variables|GameData|GetSpawnDelayMax)))))
      (Variables|Default|SetSpawnTimerHandle _returnvalue))) ; only allow 1 timer to spawn
)

(event Custom|SkydropSpawnNoGate
  (params (Transform :type Transform))
  ;[1 1 1] Force spawn IcicleLarva skydrop, bypassing gate
  (Variables|Default|SetCustomTransform Transform)
  ;;[1 1 1] has 6 second life plays audio and has physical location
  ;[1 1 1] PreBannerThunder
  (Game|SpawnBuildingActorfromClass "/IcicleLarvaGameplay/Gameplay/SkyDrop/BGA_PreSpawnHolder.BGA_PreSpawnHolder_C" (Variables|Default|GetCustomTransform) 0 false 0 0)
  (Utilities|FlowControl|Delay 1.75)
  (CallFunction|IcicleLarvaItemSpawnedClient)
  (Utilities|FlowControl|Delay (ScalableFloat|GetValueatLevel (Variables|Default|GetSpawnDelay)))
  (bind _returnvalue (Game|SpawnBuildingActorfromClass "/IcicleLarvaGameplay/Gameplay/SkyDrop/BGA_IcicleLarva_Skydrop.BGA_IcicleLarva_Skydrop_C" (Variables|Default|GetCustomTransform) 0 false "0.000000,0.000000,-1.000000" "0.000000,0.000000,0.000000" 0 0))
  ;[1 0.402 0.956 0.5] Create and attach Minimap Actor
  (bind _returnvalue_2 (Game|SpawnActorfromClass "/IcicleLarvaGameplay/Gameplay/BP_IcicleLarva_MinimapActor.BP_IcicleLarva_MinimapActor_C" (Math|Transform|MakeTransform (Transformation|GetActorLocation _returnvalue)) "Undefined" "MultiplyWithRoot" 0 0))
  (bind _output_get (Variables|Minimap|SetIcicleLarva_MinimapActor _returnvalue_2))
  (Transformation|AttachActorToActor _output_get _returnvalue "None" "SnapToTarget" "KeepWorld" "KeepWorld" true))

(event Custom|SkydropSpawnAtTransform
  (params (Transform :type Transform))
  (Variables|Default|SetCustomTransform Transform)
  (Utilities|FlowControl|Gate false
    (:Exit
      ;;[1 1 1] has 6 second life plays audio and has physical location
      ;[1 1 1] PreBannerThunder
      (Game|SpawnBuildingActorfromClass "/IcicleLarvaGameplay/Gameplay/SkyDrop/BGA_PreSpawnHolder.BGA_PreSpawnHolder_C" (Variables|Default|GetCustomTransform) 0 false 0 0)
      (Utilities|FlowControl|Delay 1.75)
      (CallFunction|IcicleLarvaItemSpawnedClient)
      (Utilities|FlowControl|Delay (ScalableFloat|GetValueatLevel (Variables|Default|GetSpawnDelay)))
      (bind _returnvalue (Game|SpawnBuildingActorfromClass "/IcicleLarvaGameplay/Gameplay/SkyDrop/BGA_IcicleLarva_Skydrop.BGA_IcicleLarva_Skydrop_C" (Variables|Default|GetCustomTransform) 0 false "0.000000,0.000000,-1.000000" "0.000000,0.000000,0.000000" 0 0))
      ;[1 0.402 0.956 0.5] Create and attach Minimap Actor
      (bind _returnvalue_2 (Game|SpawnActorfromClass "/IcicleLarvaGameplay/Gameplay/BP_IcicleLarva_MinimapActor.BP_IcicleLarva_MinimapActor_C" (Math|Transform|MakeTransform (Transformation|GetActorLocation _returnvalue)) "Undefined" "MultiplyWithRoot" 0 0))
      (bind _output_get (Variables|Minimap|SetIcicleLarva_MinimapActor _returnvalue_2))
      (Transformation|AttachActorToActor _output_get _returnvalue "None" "SnapToTarget" "KeepWorld" "KeepWorld" true))))

(event Custom|Bind%20SkyDrop%20Event
  (bind (_asynctaskproxy _payload) (GameplayEventRouter|StartListeningToStatefulEvent (GameplayEventRouter|GetGlobalEventRouter) "/IcicleLarvaGameplay/Gameplay/EventRouter/IcicleLarva_Event_Struct.IcicleLarva_Event_Struct")
    (:OnEventReceived
      _)
    (:OnSavedEventStateExists
      _))
  (if (Utilities|Struct|BreakIcicleLarvaEventStruct _payload)
    _)
  (Utilities|FlowControl|Gate false
    (:Exit
      ;;[1 1 1] has 6 second life plays audio and has physical location
      ;[1 1 1] PreBannerThunder
      (Game|SpawnBuildingActorfromClass "/IcicleLarvaGameplay/Gameplay/SkyDrop/BGA_PreSpawnHolder.BGA_PreSpawnHolder_C" (Variables|Default|GetCustomTransform) 0 false 0 0)
      (Utilities|FlowControl|Delay 1.75)
      (CallFunction|IcicleLarvaItemSpawnedClient)
      (Utilities|FlowControl|Delay (ScalableFloat|GetValueatLevel (Variables|Default|GetSpawnDelay)))
      (bind _returnvalue (Game|SpawnBuildingActorfromClass "/IcicleLarvaGameplay/Gameplay/SkyDrop/BGA_IcicleLarva_Skydrop.BGA_IcicleLarva_Skydrop_C" (Variables|Default|GetCustomTransform) 0 false "0.000000,0.000000,-1.000000" "0.000000,0.000000,0.000000" 0 0))
      ;[1 0.402 0.956 0.5] Create and attach Minimap Actor
      (bind _returnvalue_2 (Game|SpawnActorfromClass "/IcicleLarvaGameplay/Gameplay/BP_IcicleLarva_MinimapActor.BP_IcicleLarva_MinimapActor_C" (Math|Transform|MakeTransform (Transformation|GetActorLocation _returnvalue)) "Undefined" "MultiplyWithRoot" 0 0))
      (bind _output_get (Variables|Minimap|SetIcicleLarva_MinimapActor _returnvalue_2))
      (Transformation|AttachActorToActor _output_get _returnvalue "None" "SnapToTarget" "KeepWorld" "KeepWorld" true))))

(event Custom|SpawnItem
  ;[1 1 1] Spawn Timer
  (CallFunction|SkydropSpawnatTransform (Math|Transform|MakeTransform (Utilities|Array|RandomArrayItem (Default|GetPossibleSpawnLocations)) (0.0 0.0 0.0) "1.000000,1.000000,1.000000")))

(event Custom|IcicleLarvaItemSpawned_Client
  ;[1 0.402 0.956 0.502] Item Spawned - UI Client events
  (Networking|CanExecuteCosmeticEvents
    (:True
      (Default|CallOnIcicleLarvaItemSpawned) ; For UI Notification events
    )))

(event Custom|UpdateIcicleLarvaItemClaimed
  (params (PawnThatClaimedItem :type Fort%20Pawn%20Object%20Reference))
  (bind _iciclelarvaplayerpawn (Variables|Default|GetIcicleLarvaPlayerPawn))
  ;[1 0.402 0.956 0.502] UI Events - Item Claimed/Defeated
  (bind _output_get (Variables|Default|SetIcicleLarvaPlayerPawn PawnThatClaimedItem))
  (Variables|Default|SetIcicleLarvaPlayerState (Variables|/Script/Engine.Pawn|GetPlayerState _output_get))
  ;;[0.15 0.15 0.15 0.502] Player Health Updated - Replicate for UI binds
  (Networking|SwitchHasAuthority
    (:Authority
      (Health|BindEventtoOnPawnHealthChanged _iciclelarvaplayerpawn (delegate Custom|OnPlayerHealthUpdated))
      ;;[0.15 0.15 0.15 0.502] Icicle Larva Damaged - replicate to show on enemy boon progress bar
      (EventDispatchers|BindEventtoOnDamaged _iciclelarvaplayerpawn (delegate Custom|OnDamaged))
      ;;[0.15 0.15 0.15 0.502] The system this minimap icons use, store its data on server thus the SetMinimapIcon node only execute on sever.
      ;;
      ;; Client call for UI Notifications
      (CallFunction|SetIcicleLarvaItemClaimedClient))) ; Catches relevancy retriggering events
)

(event Custom|SetIcicleLarvaItemClaimed_Client
  ;[1 0.402 0.956 0.502] UI Events - Item Claimed/Defeated
  ;;[0.15 0.15 0.15 0.502] The system this minimap icons use, store its data on server thus the SetMinimapIcon node only execute on sever.
  ;;
  ;; Client call for UI Notifications
  (Default|CallOnIcicleLarvaItemClaimed (Variables|Default|GetIcicleLarvaPlayerPawn) (Variables|Default|GetIcicleLarvaPlayerState)))

(event Custom|OnPlayerHealthUpdated
  ;[1 0.402 0.956 0.502] UI Events - Item Claimed/Defeated
  ;;[0.15 0.15 0.15 0.502] Player Health Updated - Replicate for UI binds
  (Variables|Health|SetHealthPercent (Math|Float|SafeDivide (FortPawn|GetHealth (Variables|Default|GetIcicleLarvaPlayerPawn)) (FortPawn|GetMaxHealth (Variables|Default|GetIcicleLarvaPlayerPawn)))))

(event Custom|OnDamaged
  (params (DamagedActor :type Actor%20Object%20Reference) (Damage :type Float%20%28single-precision%29) (InstigatedBy :type Controller%20Object%20Reference) (DamageCauser :type Actor%20Object%20Reference) (HitLocation :type Vector) (FHitComponent :type Primitive%20Component%20Object%20Reference) (BoneName :type Name) (Momentum :type Vector))
  ;[1 0.402 0.956 0.502] UI Events - Item Claimed/Defeated
  ;;[0.15 0.15 0.15 0.502] Icicle Larva Damaged - replicate to show on enemy boon progress bar
  (Variables|Damage|SetDamage Damage)
  (Variables|Damage|SetInstigatedBy InstigatedBy)
  (CallFunction|OnDamagedClient))

(event Custom|OnDamaged_Client
  ;[1 0.402 0.956 0.502] UI Events - Item Claimed/Defeated
  ;;[0.15 0.15 0.15 0.502] Icicle Larva Damaged - replicate to show on enemy boon progress bar
  (Networking|CanExecuteCosmeticEvents
    (:True
      (Default|CallOnIcicleLarvaPawnDamaged (Variables|Damage|GetDamage) (Variables|Damage|GetInstigatedBy)))))

(event Custom|UpdateIcicleLarvaItemDefeated
  ;[1 0.402 0.956 0.502] UI Events - Item Claimed/Defeated
  ;;[0.15 0.15 0.15 0.502] Destroy the minimap actor and play Defeated Notifications
  (Actor|DestroyActor (Variables|Minimap|GetIcicleLarva_MinimapActor))
  (Networking|SwitchHasAuthority
    (:Authority
      (CallFunction|UpdateIcicleLarvaItemDefeatedClient)
      (Default|CallOnIcicleLarvaPawnEnded))) ; Passes only when the item has \nbeen actually destroyed
)

(event Custom|UpdateIcicleLarvaItemDefeated_Client
  ;[1 0.402 0.956 0.502] UI Events - Item Claimed/Defeated
  ;;[0.15 0.15 0.15 0.502] Destroy the minimap actor and play Defeated Notifications
  (Networking|CanExecuteCosmeticEvents
    (:True
      (Default|CallOnIcicleLarvaItemDestroyed))))

(fn ClearIcicleLarva
  (bind _asfort_player_controller_athena (Utilities|Casting|CastToFortPlayerControllerAthena (Pawn|GetController (Variables|Default|GetIcicleLarvaPlayerPawn))))
  (Inventory|RemoveItemfromInventoryOwner _asfort_player_controller_athena "/IcicleLarvaGameplay/Gameplay/AGID_IcicleLarva_Special.AGID_IcicleLarva_Special"))

(fn GetPossibleSpawnLocations
  (for _array_element (Variables|GameData|GetSpawnLocationsData)
    (Utilities|Array|Add (LocalVariables|GetSpawnLocations) (Math|Vector|MakeVector (ScalableFloat|GetValueatLevel _array_element) (ScalableFloat|GetValueatLevel _array_element 1.0) (ScalableFloat|GetValueatLevel _array_element 2.0))))
  (if (and (Variables|Default|GetDEBUG_SpawnOnPlayerInEditor) (Editor|InEditororPIE))
    (return (Utilities|Array|MakeArray "-5350.300000,6442.740900,20000.000000"))
    (else
      (return (LocalVariables|GetSpawnLocations)))))

(fn OnRep_HealthPercent
  (Networking|CanExecuteCosmeticEvents
    (:True
      (Default|CallOnIcicleLarvaPowerLevelUpdated (Variables|Health|GetHealthPercent)))))

(fn OnRep_IcicleLarvaLocation
  (bind _iciclelarva_minimapactor (Variables|Minimap|GetIcicleLarva_MinimapActor))
  (Utilities|IsValid _iciclelarva_minimapactor
    (:Is%20Valid
      (Transformation|SetActorLocation _iciclelarva_minimapactor (Variables|Minimap|GetIcicleLarvaLocation) false false))))

(fn OnRep_IcicleLarvaPlayerPawn
  (Networking|CanExecuteCosmeticEvents
    (:True
      (Default|CallOnIcicleLarvaPawnRelevant (Variables|Default|GetIcicleLarvaPlayerPawn)))))

(fn RollEventSpawnChance
  (bind _spawnchancepercentage (Variables|GameData|GetSpawnChancePercentage))
  (bind _returnvalue (ScalableFloat|GetValueatLevel _spawnchancepercentage))
  (bind _output_get (Variables|Default|SetRandomSpawnRoll (Math|Random|RandomFloat)))
  (bind _output_get_2 (Variables|Default|SetIsEventActiveThisMatch (< _output_get _returnvalue)))
  (return _output_get_2 _output_get _returnvalue))
