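; Pawn component that turns a player into Icicle Larva, an enlarged, cosmetically swapped form with team-size-scaled health and timed damage drain.
; BeginPlay plays or skips the intro montage, whose timer activates the flying movement mode and drain.
; Tag listeners switch ground and Heat Vision cameras, while replicated bools cover relevancy and missed notifies.
; Water is blocked to relaunch flight, and EndPlay restores the pawn and awards Hero Rank assists.
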(event EventBeginPlay
  (bind _returnvalue (Components|GetOwner))
  (bind _ownerplayerpawn (Variables|Default|GetOwnerPlayerPawn))
  (Utilities|IsValid _returnvalue
    (:Is%20Valid
      (bind _asfort_player_pawn (Utilities|Casting|CastToFortPlayerPawn _returnvalue))
      (Variables|Default|SetOwnerPlayerPawn _asfort_player_pawn)
      (CallFunction|SetHealthbyPlaylist)
      (Utilities|FlowControl|Sequence
        (:then_0
          ;[1 1 1] For a pawn entering relevancy this replicated bool might already be true
          (Variables|Default|GetbIsTransformed
            (:then
              (Default|TransformPawn))
            (:else
              (if (CallFunction|IsIntroSequenceSkipped)
                ;[1 1 1] If intro is skipped, immediately transform pawn and enable camera modifier that would normally be activated on intro montage end
                ; (we don't apply the damage timer GE for easier testing)
                (Default|TransformPawn)
                (CallFunction|UpdateGroundCameraModifier)
                (else
                  (CallFunction|PlayIntroSequence))))))
        (:then_1
          ;[1 1 1] Bind OnDamage Event for Hero Rank Point rewards for players that engage in the fight & apply Hero Rank multiplier
          (EventDispatchers|BindEventtoOnDamaged (Variables|Default|GetOwnerPlayerPawn) (EventDispatchers|CreateEvent :!event Custom|OnIcicleLarvaDamaged))
          (CallFunction|ApplyHeroRankMultipliertoOwner))
        (:then_2
          ;[1 1 1] Handle Heat Vision camera logic here so spectating players still have correct modifiers applied
          (Ability|Tasks|WaitGameplayTagAddToActor (Variables|Default|GetOwnerPlayerPawn) (Variables|Default|GetHeatVisionBeamTag)
            (:then
              (Ability|Tasks|WaitGameplayTagRemoveFromActor (Variables|Default|GetOwnerPlayerPawn) (Variables|Default|GetHeatVisionBeamTag)
                (:Removed
                  (CallFunction|DisableHeatVisionCameraModifier))))
            (:Added
              (CallFunction|EnableHeatVisionCameraModifier))))
        (:then_3
          ;[1 1 1] Update always targeting based on hover/grounded state
          (Networking|CanExecuteCosmeticEvents
            (:True
              (Ability|Tasks|WaitGameplayTagCountChangedOnActor _ownerplayerpawn (Variables|Default|GetMovementStateHoverTag)
                (:then
                  (Ability|Tasks|WaitGameplayTagCountChangedOnActor _ownerplayerpawn (Variables|Default|GetMovementModeIcicleLarvaTag)
                    (:TagCountChanged
                      _)))
                (:TagCountChanged
                  _))
              (CallFunction|RefreshAlwaysTargeting))))))))

(event Custom|SetHealthbyPlaylist
  ;[1 1 1] Health Scaling Logic
  ;;[1 1 1] Getting the team size from the playlist to set the team size variable
  (bind (_gamestate _playlist) (Class|FortAsyncActionCurrentPlaylistReady|PlaylistDataReadyAsync)
    (:OnReady
      (Variables|Default|SetPlaylistTeamSize (Math|Integer|Clamp(Integer) (Variables|/Script/FortniteGame.FortPlaylist|GetMaxTeamSize _playlist) 1 4))
      (CallFunction|ScaleHealthbyTeamSize))))

(event Custom|ScaleHealthByTeamSize
  ;[1 1 1] Health Scaling Logic
  ;;[1 1 1] Setting Health Values based on the TeamSize
  (bind _returnvalue (GameplayEffects|ApplyGameplayEffectToSelf :self (Ability|GetAbilitySystemComponent (Variables|Default|GetOwnerPlayerPawn)) :GameplayEffectClass (Variables|Default|GetGE_ScaleHealthByTeamSize) :Level (Math|Conversions|ToFloat(Integer) (- (Variables|Default|GetPlaylistTeamSize) 1))))
  (Variables|Default|SetScaleHealthGEHandle _returnvalue))

(event Custom|ResetPlayer
  (bind _returnvalue (Components|GetOwner))
  (bind _ownerplayerpawn (Variables|Default|GetOwnerPlayerPawn))
  (bind _returnvalue_2 (Ability|GetAbilitySystemComponent _ownerplayerpawn))
  (bind _ownerplayerpawn_2 (Variables|Default|GetOwnerPlayerPawn))
  (Utilities|IsValid _returnvalue
    (:Is%20Valid
      (bind _asfort_player_pawn (Utilities|Casting|CastToFortPlayerPawn _returnvalue))
      (CallFunction|ResetPawnCosmetics _asfort_player_pawn)
      (CallFunction|ResizePawn _asfort_player_pawn "1.000000,1.000000,1.000000")
      (GameplayEffects|RemoveActiveGameplayEffect _returnvalue_2 (Variables|Default|GetTimerGEHandle) -1)
      (GameplayEffects|RemoveActiveGameplayEffect _returnvalue_2 (Variables|Default|GetScaleHealthGEHandle) -1)
      (Collision|SetCollisionResponsetoChannel (Variables|/Script/Engine.Character|GetCapsuleComponent _ownerplayerpawn_2) (Variables|Default|GetWaterChannelToBlock) (Variables|Default|GetWaterChannelOriginalResponse))
      (EventDispatchers|UnbindEventfromOnBaseChangedDelegate _ownerplayerpawn_2 (EventDispatchers|CreateEvent :!event HandleBaseChanged)))))

(event EventEndPlay
  (params (EndPlayReason :type EEndPlayReason%20Enum :default "Destroyed"))
  (Utilities|Time|ClearandInvalidateTimerbyHandle (Variables|Default|GetOnMontageFinishedTimerHandle))
  ;[1 1 1] Always reset player when this component is removed
  (CallFunction|ResetPlayer)
  ;[1 1 1] Always Remove the Intro GEs when component is removed
  (CallFunction|RemoveIntroGEs)
  (CallFunction|RemoveLocalIntroTag)
  ;[1 1 1] Clean up camera
  (CallFunction|RemoveIntroCameraMode)
  (CallFunction|DisableGroundCameraModifier)
  (CallFunction|DisableHeatVisionCameraModifier)
  ;[0.002033 0.18 0.8594 0.5] Award Hero points and unbind onDamaged event. Unset multiplier on iciclelarva player
  (CallFunction|AwardHeroPointsforAssistParticipants)
  (EventDispatchers|UnbindEventfromOnDamaged (Variables|Default|GetOwnerPlayerPawn) (EventDispatchers|CreateEvent :!event Custom|OnIcicleLarvaDamaged))
  (CallFunction|RemoveHeroRankMultipliertoOwner)
  (|CallParentFunction EndPlayReason))

(event Custom|PlayIntroSequence
  (bind _ownerplayerpawn (Variables|Default|GetOwnerPlayerPawn))
  (bind _returnvalue (Ability|GetAbilitySystemComponent _ownerplayerpawn))
  (bind _ownerplayerpawn_2 (Variables|Default|GetOwnerPlayerPawn))
  ;[1 1 1] Sends gameplay event to remove Balloons from the players as a fix to FORT-905119
  (Ability|SendGameplayEventtoActor (Variables|Default|GetOwnerPlayerPawn) (Variables|Default|GetEvent%20Tag%20Release%20Balloons))
  (CallFunction|BlockWaterInteraction)
  ;[1 1 1] Prevent movement/firing during montage
  (bind _returnvalue_2 (GameplayEffects|ApplyGameplayEffectToSelf (Ability|GetAbilitySystemComponent (Variables|Default|GetOwnerPlayerPawn)) (Variables|Default|GetIntroSequencePassiveGEClass) 0.0 (Class|AthenaFunctionLibrary|MakeSimpleFortEffectContext (Variables|Default|GetOwnerPlayerPawn))))
  (Variables|Default|SetPreventMoveGEHandle _returnvalue_2)
  ;[1 1 1] Add intro tag to make sure client has it immediately
  (Ability|GameplayEffect|AddLooseGameplayTags (Variables|Default|GetOwnerPlayerPawn) (GameplayTags|MakeLiteralGameplayTagContainer "(GameplayTags=((TagName=\"Granted.IcicleLarva.Intro\")))"))
  ;[1 1 1] Set movement mode to Flying so root motion could move us around
  (Pawn|Components|CharacterMovement|SetMovementMode (Variables|/Script/Engine.Character|GetCharacterMovement (Variables|Default|GetOwnerPlayerPawn)) "MOVE_Flying" 0)
  (bind _returnvalue_3 (Animation|Utilities|PlayAnimMontage (Variables|Default|GetOwnerPlayerPawn) (Variables|Default|GetIntroAnimMontage)))
  ;[1 1 1] Adding GE effects before setting timer due to infrequent potential race condition that removes invuln before applying it.
  (Utilities|FlowControl|Sequence
    (:then_0
      ;[1 1 1] Adding DamageImmunity GE (Hotfixable)
      (if (ScalableFloat|GetValueasBool (Variables|Default|GetRow_ShouldApplyDamageImmunity))
        (bind _returnvalue_4 (Ability|GameplayEffect|AssignTagSetbyCallerMagnitude (GameplayEffects|MakeOutgoingSpec _returnvalue (Variables|Default|GetDamageImmunityGEClass) 0.0 (Class|AthenaFunctionLibrary|MakeSimpleFortEffectContext _ownerplayerpawn)) "(TagName=\"SetByCaller.IcicleLarva.SequenceDuration\")" _returnvalue_3))
        (bind _returnvalue_5 (GameplayEffects|ApplyGameplayEffectSpecToSelf _returnvalue _returnvalue_4))
        (Variables|Default|SetDamageImmunityGEHandle _returnvalue_5)
        (else
          _))
      ;[1 1 1] Set replicated bool in case this pawn enters relevancy after the transformation
      (Networking|SwitchHasAuthority
        (:Authority
          (Variables|Default|SetbIsTransformed true))
        (:Remote
          (CallFunction|PlayIntroMusic)
          (bind _asfort_player_camera_base (Utilities|Casting|CastToFortPlayerCameraBase (Variables|/Script/Engine.PlayerController|GetPlayerCameraManager (Fortnite|GetFortPlayerControllerfromActor _ownerplayerpawn_2))))
          (Camera|AddCameraModeOverride _asfort_player_camera_base (Variables|Default|GetIntroCameraMode) (Variables|Default|GetIntroCameraModeTag) 50 self _ownerplayerpawn_2))))
    (:then_1
      ;[1 1 1] When the intro sequence is most of the way through activate the MME
      ; Note: Not using a notify to avoid it being missed when the montage blends out
      (bind _returnvalue_6 (Utilities|Time|SetTimerbyEvent (delegate Custom|OnMontageFinished) (* _returnvalue_3 (ScalableFloat|GetValueatLevel (Variables|Default|GetRow_IntroMontagePercentToActivateMME)))))
      (Variables|Default|SetOnMontageFinishedTimerHandle _returnvalue_6))))

(event Custom|OnMontageFinished
  (Class|FortMovementModeExtLogicFlyingIcicleLarva|TryActivateIcicleLarvaMME (Variables|Default|GetOwnerPlayerPawn) (Variables|FortPawnComponent_IcicleLarva|GetMMEClass) (Utilities|Struct|MakeFortMovementModeIcicleLarvaCreationData 0.0))
  (CallFunction|RemoveIntroGEs)
  (CallFunction|ApplyTimer)
  (CallFunction|UpdateLocalIntroTag)
  (Networking|SwitchHasAuthority
    (:Remote
      ;[1 1 1] Delay camera removal by 1 tick to make sure the MME has already been activated
      (Utilities|Time|SetTimerforNextTickbyEvent (delegate Custom|UpdateCameraStatePostIntro)))))

(event Custom|UpdateCameraStatePostIntro
  (CallFunction|UpdateGroundCameraModifier)
  (CallFunction|RemoveIntroCameraMode))

(event Custom|RemoveIntroGEs
  (bind _ownerplayerpawn (Variables|Default|GetOwnerPlayerPawn))
  (bind _returnvalue (Ability|GetAbilitySystemComponent _ownerplayerpawn))
  ;[1 1 1] Removing the GEs applied to the player during Intro Sequence
  (GameplayEffects|RemoveActiveGameplayEffect _returnvalue (Variables|Default|GetPreventMoveGEHandle) -1)
  (GameplayEffects|RemoveActiveGameplayEffect _returnvalue (Variables|Default|GetDamageImmunityGEHandle) -1))

(event Custom|PlayIntroMusic
  (bind _ownerplayerpawn (Variables|Default|GetOwnerPlayerPawn)
    (:then
      (if (Pawn|IsLocallyViewed _ownerplayerpawn)
        (bind _returnvalue (Audio|SpawnSound2D (Variables|Default|GetIntroSound)))
        (bind _output_get (Variables|Default|SetMusicIntroAudio _returnvalue))
        (Audio|AddEventbyComponent (GameInstanceSubsystems|GetMusicEventSubsystem) _output_get "(TagName=\"MusicEventSubsystem.Event.Gameplay.Ability.1P\")" "(GameplayTags=)" 0 "")))
    (:else)))

(event Custom|UpdateGroundCameraModifier
  ;[1 1 1] If not in MME, enable ground camera modifier
  (Utilities|SwitchHasAuthority
    (:Remote
      (if (not (GameplayTags|HasMatchingGameplayTag (Variables|Default|GetOwnerPlayerPawn) (Variables|Default|GetIcicleLarvaMovementTag)))
        (CallFunction|EnableGroundCameraModifier))
      ;[1 1 1] Listen for MME tag changes to enable/disable modifier appropriately
      (Ability|Tasks|WaitGameplayTagAddToActor (Variables|Default|GetOwnerPlayerPawn) (Variables|Default|GetIcicleLarvaMovementTag)
        (:then
          (Ability|Tasks|WaitGameplayTagRemoveFromActor (Variables|Default|GetOwnerPlayerPawn) (Variables|Default|GetIcicleLarvaMovementTag)
            (:Removed
              (CallFunction|EnableGroundCameraModifier))))
        (:Added
          (CallFunction|DisableGroundCameraModifier))))))

(event Custom|EnableGroundCameraModifier
  (for _array_element (FortPawn|GetAllLocalViewingPlayerControllers (Variables|Default|GetOwnerPlayerPawn))
    (CallFunction|EnableCameraModifierforController _array_element (Variables|Default|GetGroundCameraModifierClass))))

(event Custom|DisableGroundCameraModifier
  (for _array_element (FortPawn|GetAllLocalViewingPlayerControllers (Variables|Default|GetOwnerPlayerPawn))
    (CallFunction|DisableCameraModifierforController _array_element (Variables|Default|GetGroundCameraModifierClass))))

(event Custom|UpdateLocalIntroTag
  (bind _row_removelocalintrotagdelay (Variables|Default|GetRow_RemoveLocalIntroTagDelay))
  (bind _returnvalue (ScalableFloat|GetValueatLevel _row_removelocalintrotagdelay))
  (if (> _returnvalue 0)
    (Utilities|Time|SetTimerbyEvent (delegate Custom|RemoveLocalIntroTag) _returnvalue)
    (else
      (CallFunction|RemoveLocalIntroTag))))

(event Custom|EnableHeatVisionCameraModifier
  (for _array_element (FortPawn|GetAllLocalViewingPlayerControllers (Variables|Default|GetOwnerPlayerPawn))
    (CallFunction|EnableCameraModifierforController _array_element (select (GameplayTags|HasMatchingGameplayTag (Variables|Default|GetOwnerPlayerPawn) (Variables|Default|GetIcicleLarvaMovementTag)) (Variables|Default|GetHeatVision_Air_CameraModifierClass) (Variables|Default|GetHeatVision_Ground_CameraModifierClass))))
  ;[1 1 1] Handle case where the player lands while firing this ability
  (bind _asynctaskproxy (Ability|Tasks|WaitGameplayTagRemoveFromActor (Variables|Default|GetOwnerPlayerPawn) (Variables|Default|GetIcicleLarvaMovementTag))
    (:then
      (Variables|Default|SetWaitLandedDuringHeatVisionAsync _asynctaskproxy))
    (:Removed
      (for _array_element_2 (FortPawn|GetAllLocalViewingPlayerControllers (Variables|Default|GetOwnerPlayerPawn))
        (CallFunction|DisableCameraModifierforController _array_element_2 (Variables|Default|GetHeatVision_Air_CameraModifierClass))
        (CallFunction|EnableCameraModifierforController _array_element_2 (Variables|Default|GetHeatVision_Ground_CameraModifierClass))))))

(event Custom|RemoveLocalIntroTag
  (Ability|GameplayEffect|RemoveLooseGameplayTags (Variables|Default|GetOwnerPlayerPawn) (GameplayTags|MakeLiteralGameplayTagContainer "(GameplayTags=((TagName=\"Granted.IcicleLarva.Intro\")))")))

(event Custom|DisableHeatVisionCameraModifier
  (Ability|Async|EndAction (Variables|Default|GetWaitLandedDuringHeatVisionAsync))
  (for _array_element (FortPawn|GetAllLocalViewingPlayerControllers (Variables|Default|GetOwnerPlayerPawn))
    (CallFunction|DisableCameraModifierforController _array_element (Variables|Default|GetHeatVision_Ground_CameraModifierClass))
    (CallFunction|DisableCameraModifierforController _array_element (Variables|Default|GetHeatVision_Air_CameraModifierClass))))

(event EventOnOwnerPawnBecameViewTarget
  (params (PC :type Player%20Controller%20Object%20Reference))
  (if (not (GameplayTags|HasMatchingGameplayTag (Variables|Default|GetOwnerPlayerPawn) (Variables|Default|GetIcicleLarvaMovementTag)))
    (CallFunction|EnableCameraModifierforController PC (Variables|Default|GetGroundCameraModifierClass))))

(event EventOnOwnerPawnEndViewTarget
  (params (PC :type Player%20Controller%20Object%20Reference))
  (CallFunction|DisableCameraModifierforController PC (Variables|Default|GetGroundCameraModifierClass)))

(event Custom|RemoveIntroCameraMode
  (bind _asfort_player_camera_base (Utilities|Casting|CastToFortPlayerCameraBase (Variables|/Script/Engine.PlayerController|GetPlayerCameraManager (Fortnite|GetFortPlayerControllerfromActor (Variables|Default|GetOwnerPlayerPawn)))))
  (Camera|RemoveCameraModeOverride _asfort_player_camera_base (Variables|Default|GetIntroCameraModeTag) self true))

(event Custom|OnIcicleLarvaDamaged
  (params (DamagedActor :type Actor%20Object%20Reference) (Damage :type Float%20%28single-precision%29) (InstigatedBy :type Controller%20Object%20Reference) (DamageCauser :type Actor%20Object%20Reference) (HitLocation :type Vector) (FHitComponent :type Primitive%20Component%20Object%20Reference) (BoneName :type Name) (Momentum :type Vector))
  ;[0.002033 0.18 0.8594 0.5] Add damage participant players to array so we can award them hero rank point event on end play / icile larva defeated
  (bind _asfort_player_pawn (Utilities|Casting|CastToFortPlayerPawn (Pawn|GetControlledPawn InstigatedBy)))
  (if (not (== DamagedActor _asfort_player_pawn))
    (Utilities|Array|AddUnique (Variables|HeroRank|GetDamaged%20Assisted%20Player%20Pawn%20Array) _asfort_player_pawn)))

(fn ApplyHeroRankMultiplierToOwner
  ;[0.15 0.1232 0.007516 0.5] Applying initial multiplier of +1 (For a total of 2x) on change. Clearing when exiting
  (Utilities|SwitchHasAuthority
    (:Authority
      (bind _returnvalue (Class|CherrySmokeHeroRankingManager|PushHeroScoreMultiplier (Actor|GetComponentbyClass (Variables|/Script/Engine.Pawn|GetPlayerState (Variables|Default|GetOwnerPlayerPawn)) "/Script/CherrySmokeHeroRankingSystemRuntime.CherrySmokeHeroRankingManager") (Utilities|Struct|MakeHeroRankMultiplier 2.0)))
      (Variables|HeroRank|SetHeroScoreMultiplierHandle _returnvalue))))

(fn ApplyTimer
  (bind _ownerplayerpawn (Variables|Default|GetOwnerPlayerPawn))
  (bind _returnvalue (Ability|GetAbilitySystemComponent _ownerplayerpawn))
  (if (CallFunction|IsDamageTimerEnabled)
    ;[1 1 1] Applying a Timer GE that deals X damage to the player every second
    (bind _returnvalue_2 (Ability|GameplayEffect|AssignTagSetbyCallerMagnitude (GameplayEffects|MakeOutgoingSpec _returnvalue (Variables|Default|GetTimerGEClass) 0.0 (Class|AthenaFunctionLibrary|MakeSimpleFortEffectContext _ownerplayerpawn)) "(TagName=\"SetByCaller.IcicleLarva.Timer\")" (/ (FortPawn|GetMaxHealth _ownerplayerpawn) (Math|Float|Max(Float) (ScalableFloat|GetValueatLevel (Variables|Default|GetRow_TimerTotalDuration)) 1.0))))
    (bind _returnvalue_3 (GameplayEffects|ApplyGameplayEffectSpecToSelf _returnvalue _returnvalue_2))
    (Variables|Default|SetTimerGEHandle _returnvalue_3)))

(fn AwardHeroPointsForAssistParticipants
  (bind _damaged_assisted_player_pawn_array (Variables|HeroRank|GetDamaged%20Assisted%20Player%20Pawn%20Array))
  (bind _returnvalue (Utilities|Array|Length _damaged_assisted_player_pawn_array))
  (bind _asfort_game_state_athena (Utilities|Casting|CastToFortGameStateAthena (Game|GetGameState)))
  (LocalVariables|SetFort%20Game%20State%20Athena _asfort_game_state_athena)
  (bind _array_element (Utilities|Array|ForEachLoopwithBreak (Variables|HeroRank|GetDamaged%20Assisted%20Player%20Pawn%20Array))
    (:LoopBody
      ;[1 1 1] Flag dead players for removal
      (if (not (FortPawn|IsDead _array_element))
        (bind _asfort_player_state_athena (Utilities|Casting|CastToFortPlayerStateAthena (Variables|/Script/Engine.Pawn|GetPlayerState _array_element)))
        ;[1 1 1] We've found a valid player to be the point award recipient. Remove the squad from the damage array
        (bind _returnvalue_2 (Class|FortGameStateAthena|GetSquadMemberPlayerStates (LocalVariables|GetFort%20Game%20State%20Athena) _asfort_player_state_athena true))
        (for _array_element_2 _returnvalue_2
          (LocalVariables|SetPlayerToAwardHeroPointsTo (Game|Player|GetPlayerPawn (Fortnite|GetFortPlayerControllerfromActor _array_element_2)))
          (Utilities|Array|RemoveItem (Variables|HeroRank|GetDamaged%20Assisted%20Player%20Pawn%20Array) (Game|Player|GetPlayerPawn (Fortnite|GetFortPlayerControllerfromActor _array_element_2))))
        ;[1 1 1] Award points to a valid player. 1 player shares with squad so we don't award points based on squad size
        ; Send break to for loop
        (bind _playertoawardheropointsto (LocalVariables|GetPlayerToAwardHeroPointsTo)
          (:then
            (CustomVerb|BroadcastCustomVerbMessage (CustomVerb|MakeCustomVerbMessage _playertoawardheropointsto (Variables|HeroRank|GetIcicleLarva%20Assisted%20Defeat%20Verb%20Name))))
          (:else)))
      (Utilities|Array|AddUnique (LocalVariables|GetPlayersToRemoveFromArray) _array_element))
    (:Completed
      _))
  ;[1 1 1] If we still have players in array, remove dead players & run again
  ; resizing an array( from removing squad ) will cause the for loop to not complete with all entries
  ; If the size is the same as last time, end
  (for _array_element_3 (LocalVariables|GetPlayersToRemoveFromArray)
    (Utilities|Array|RemoveItem (Variables|HeroRank|GetDamaged%20Assisted%20Player%20Pawn%20Array) _array_element_3))
  (if (and (> _returnvalue 0) (!= _returnvalue (LocalVariables|GetArraySize)))
    (LocalVariables|SetArraySize _returnvalue)))

(fn BlockWaterInteraction
  (bind _ownerplayerpawn (Variables|Default|GetOwnerPlayerPawn))
  (bind _capsulecomponent (Variables|/Script/Engine.Character|GetCapsuleComponent _ownerplayerpawn))
  (Variables|Default|SetWaterChannelOriginalResponse (Collision|GetCollisionResponsetoChannel _capsulecomponent (Variables|Default|GetWaterChannelToBlock)))
  (Collision|SetCollisionResponsetoChannel _capsulecomponent (Variables|Default|GetWaterChannelToBlock) "ECR_Block")
  (EventDispatchers|BindEventtoOnBaseChangedDelegate _ownerplayerpawn (EventDispatchers|CreateEvent :!event HandleBaseChanged)))

(fn ChangePawnCosmetics
  (params (Pawn :type Fort%20Player%20Pawn%20Object%20Reference))
  (Utilities|IsValid Pawn
    (:Is%20Valid
      (Variables|Default|SetPreviousCosmetic (CosmeticLoadout|GetItemDefinitionTypefromLoadoutwithActor (Variables|GetPawn) "(TagName=\"LoadoutSlot.Character\")" "/Script/FortniteGame.AthenaCharacterItemDefinition"))
      (Variables|Default|SetPreviousBackpack (CosmeticLoadout|GetItemDefinitionTypefromLoadoutwithActor (Variables|GetPawn) "(TagName=\"LoadoutSlot.Backpack\")" "/Script/FortniteGame.AthenaBackpackItemDefinition"))
      (Cosmetics|PushCosmeticOverrideOntoPawn Pawn (Variables|Default|GetIcicleLarvaCosmetic))
      (Cosmetics|PushCosmeticOverrideOntoPawn Pawn (Variables|Default|GetIcicleLarvaBackpack))
      (Class|FortPlayerAnimationInterface|SetMovementStyleOverride (Animation|GetAnimationInterface Pawn) "Walking")
      (Utilities|SwitchHasAuthority
        (:Authority
          (Variables|Default|SetbPawnCosmeticsChanged true))))))

(fn DisableCameraModifierForController
  (params (PlayerController :type Player%20Controller%20Object%20Reference) (CameraModifier :type Camera%20Modifier%20Class%20Reference))
  (bind _asfort_player_camera_base (Utilities|Casting|CastToFortPlayerCameraBase (Variables|/Script/Engine.PlayerController|GetPlayerCameraManager PlayerController)))
  (bind _returnvalue (Game|Player|FindCameraModifierbyClass _asfort_player_camera_base CameraModifier))
  (Utilities|IsValid _returnvalue
    (:Is%20Valid
      (CameraModifier|DisableModifier _returnvalue false))))

(fn EnableCameraModifierForController
  (params (PlayerController :type Player%20Controller%20Object%20Reference) (CameraModifier :type Camera%20Modifier%20Class%20Reference))
  (bind _asfort_player_camera_base (Utilities|Casting|CastToFortPlayerCameraBase (Variables|/Script/Engine.PlayerController|GetPlayerCameraManager PlayerController)))
  (bind _returnvalue (Game|Player|FindCameraModifierbyClass _asfort_player_camera_base CameraModifier))
  (Utilities|IsValid _returnvalue
    (:Is%20Valid
      (CameraModifier|EnableModifier _returnvalue))
    (:Is%20Not%20Valid
      (bind _returnvalue_2 (Game|Player|AddNewCameraModifier _asfort_player_camera_base CameraModifier))
      (CameraModifier|EnableModifier _returnvalue_2))))

(fn HandleBaseChanged
  (params (Pawn :type Fort%20Pawn%20Object%20Reference) (NewBase :type Actor%20Object%20Reference))
  ;[1 1 1] Ignore if the pawn has any of the tags that cause us to ignore movement changes. Like during the intro sequence, during a guided motion MME, etc...
  (if (not (GameplayTags|HasAnyMatchingGameplayTags Pawn (Variables|FortPawnComponent_IcicleLarva|GetIgnoreMovementChangeTags)))
    ;[1 1 1] If we're based on a water body or an invalid base, activate MME
    (if (Utilities|IsA NewBase "/Script/FortniteGame.FortWaterBodyActor")
      (Class|FortMovementModeExtLogicFlyingIcicleLarva|TryActivateIcicleLarvaMME (Variables|GetPawn) (Variables|FortPawnComponent_IcicleLarva|GetMMEClass) (Utilities|Struct|MakeFortMovementModeIcicleLarvaCreationData (ScalableFloat|GetValueatLevel (Variables|Default|GetRow_Ascend_TimeWhenLaunchingFromWater))))
      (else
        (Utilities|IsValid (Variables|GetNewBase)
          (:Is%20Not%20Valid
            (Class|FortMovementModeExtLogicFlyingIcicleLarva|TryActivateIcicleLarvaMME (Variables|GetPawn) (Variables|FortPawnComponent_IcicleLarva|GetMMEClass) (Utilities|Struct|MakeFortMovementModeIcicleLarvaCreationData (ScalableFloat|GetValueatLevel (Variables|Default|GetRow_Ascend_TimeWhenLaunchingFromNullBase))))))))))

(fn OnRep_bIsScaledUp
  ;[1 1 1] The anim notify can be missed while the actor is still in relevancy (seems like they have a different range to be considered "relevant"), so to ensure all clients receive the scale up we also scale up when the server does.
  (Utilities|SwitchHasAuthority
    (:Remote
      (Variables|Default|GetbIsScaledUp
        (:then
          (CallFunction|ScaleUpPawn))
        (:else)))))

(fn OnRep_bPawnCosmeticsChanged
  ;[1 1 1] The anim notify can be missed while the actor is still in relevancy (seems like they have a different range to be considered "relevant"), so to ensure all clients receive the cosmetic change we also change when the server does.
  (Utilities|SwitchHasAuthority
    (:Remote
      (Variables|Default|GetbPawnCosmeticsChanged
        (:then
          (CallFunction|ChangePawnCosmetics (Variables|Default|GetOwnerPlayerPawn)))
        (:else)))))

(fn RefreshAlwaysTargeting
  (Beam|SetClientAlwaysTargeting (or (GameplayTags|HasMatchingGameplayTag (Variables|Default|GetOwnerPlayerPawn) "(TagName=\"IcicleLarva.MovementState.Hover\")") (not (GameplayTags|HasMatchingGameplayTag (Variables|Default|GetOwnerPlayerPawn) "(TagName=\"MovementMode.IcicleLarva\")")))))

(fn ReInitPhysicsAsset
  (bind _ownerplayerpawn (Variables|Default|GetOwnerPlayerPawn)
    (:then
      (Components|SkinnedMesh|SetPhysicsAsset (Variables|/Script/Engine.Character|GetMesh _ownerplayerpawn) "/Game/Characters/Player/Male/Male_Avg_Base/Fortnite_M_Avg_Player_Physics.Fortnite_M_Avg_Player_Physics" true))
    (:else
      (Fortnite|FortShippingLog self "Could not re-init physics asset because OwnerPlayerPawn is invalid!" true))))

(fn RemoveHeroRankMultiplierToOwner
  ;[0.15 0.1232 0.007516 0.5] Clearing multiplier when ending
  (Utilities|SwitchHasAuthority
    (:Authority
      (Class|CherrySmokeHeroRankingManager|RemoveHeroScoreMultiplier (Actor|GetComponentbyClass (Variables|/Script/Engine.Pawn|GetPlayerState (Variables|Default|GetOwnerPlayerPawn)) "/Script/CherrySmokeHeroRankingSystemRuntime.CherrySmokeHeroRankingManager") (Variables|HeroRank|GetHeroScoreMultiplierHandle)))))

(fn ResetPawnCosmetics
  (params (Pawn :type Fort%20Player%20Pawn%20Object%20Reference))
  (Cosmetics|PushCosmeticOverrideOntoPawn Pawn (Variables|Default|GetPreviousCosmetic))
  (Utilities|IsValid (Variables|Default|GetPreviousBackpack)
    (:Is%20Valid
      (Cosmetics|PushCosmeticOverrideOntoPawn Pawn (Variables|Default|GetPreviousBackpack)))
    (:Is%20Not%20Valid
      (Cosmetics|PushCosmeticOverrideOntoPawn Pawn "/Game/Athena/Items/Cosmetics/Backpacks/BID_STWHeroNoDefaultBackpack.BID_STWHeroNoDefaultBackpack")))
  (Class|FortPlayerAnimationInterface|SetMovementStyleOverride (Animation|GetAnimationInterface Pawn) "Running"))

(fn ResizePawn
  (params (Pawn :type Fort%20Player%20Pawn%20Object%20Reference) (NewPawnScale :type Vector))
  (bind _returnvalue (Customization|GetSkeletalMeshforPartType Pawn "Backpack" :!class /Script/FortniteGame.FortPlayerPawn))
  (Transformation|SetActorScale3D Pawn NewPawnScale)
  ;[0.713 1.2e-05 0 0.5] MUST re-init physics asset after scaling, otherwise there could be holes where shots go through the body (namely neck area)
  ;
  ; Only gated behind scalable float because this is a late change
  (if (ScalableFloat|GetValueasBool (Variables|Default|GetRow_Transformation_ReinitPhysicsAsset))
    ;;[1 1 1] Delaying by a frame because this call can come from an anim notify, causing a recursion in USkeletalMeshComponent::PostAnimEvaluation
    (Utilities|Time|SetTimerforNextTickbyEvent (EventDispatchers|CreateEvent :!event ReInitPhysicsAsset))
    (else
      _))
  (Variables|/Script/Engine.SkeletalMeshComponent|SetClothGeometryScale 1.0 _returnvalue)
  (Clothing|RecreateClothingActors _returnvalue))

(fn ScaleUpPawn
  (LocalVariables|SetPlayerSize (ScalableFloat|GetValueatLevel (Variables|Default|GetRow_PlayerSize)))
  (CallFunction|ResizePawn (Variables|Default|GetOwnerPlayerPawn) (Math|Vector|MakeVector (LocalVariables|GetPlayerSize) (LocalVariables|GetPlayerSize) (LocalVariables|GetPlayerSize)))
  (Utilities|SwitchHasAuthority
    (:Authority
      (Variables|Default|SetbIsScaledUp true))))

(fn TransformPawn
  (CallFunction|ChangePawnCosmetics (Variables|Default|GetOwnerPlayerPawn))
  (CallFunction|ScaleUpPawn)
  (CallFunction|BlockWaterInteraction)
  ;[1 1 1] Set replicated bool in case this pawn enters relevancy after the transformation
  (Networking|SwitchHasAuthority
    (:Authority
      (Variables|Default|SetbIsTransformed true))))
